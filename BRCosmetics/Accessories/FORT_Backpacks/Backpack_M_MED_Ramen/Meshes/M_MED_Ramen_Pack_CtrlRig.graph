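# Control Rig for the Ramen backpack accessory that adjusts the orientation of the dyn_dno_K bone on the pack.
# A single local-space-offset aim constraint points that bone's negative X axis at its sibling dyn_dno_all, at partial weight and filtered to Z rotation only.
# The constraint does not maintain offset, and the dyn_list topping bones beneath dyn_dno_K have no logic of their own in this rig.
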ControlRigBlueprint /BRCosmetics/Accessories/FORT_Backpacks/Backpack_M_MED_Ramen/Meshes/M_MED_Ramen_Pack_CtrlRig
graphs [
 RigVMFunctionLibrary {links [], nodes []}
 RigVMModel {nodes [
   RigUnit_BeginExecution = &RigUnit_BeginExecution()
   &RigUnit_AimConstraintLocalSpaceOffset(
    AdvancedSettings=(DebugSettings=(Scale=10,WorldOffset=(Rotation=(X=0,Y=0,Z=0,W=1),Translation=(X=0,Y=0,Z=0),Scale3D=(X=1,Y=1,Z=1))),RotationOrderForFilter=XZY)
    AimAxis=(X=-1,Y=0,Z=0)
    Child=(Type=Bone,Name="dyn_dno_K")
    Filter=(bX=false,bY=false,bZ=True)
    Parents=((Item=(Type=Bone,Name="dyn_dno_all"),Weight=2))
    Weight=0.8
    WorldUp=(Target=(X=0,Y=0,Z=-1),Kind=Location,Space=(Type=None,Name="None"))
    bMaintainOffset=false
    ExecutePin=@RigUnit_BeginExecution.ExecutePin
   )
  ]}
]
hierarchy [
 root {loc=<0 0 0>, rot=<0 0 0 1>}
 pack {parent=root, loc=<0 -0.7316 0>, rot=<0 0 0 1>}
 dyn_dno_all {parent=pack, loc=<0 0 -11>, rot=<-0.0851 0.7659 0.0872 -0.6314>}
 dyn_dno_K {parent=pack, loc=<2.4326 -10.232 -8.0649>, rot=<-0.0851 0.7659 0.0872 -0.6314>}
 dyn_dno_K1 {parent=dyn_dno_K, loc=<0.98 0 0>, rot=<0 0 0 1>}
 dyn_dno_S {parent=dyn_dno_K1, loc=<1.1659 0 0>, rot=<0 0 0 1>}
 dyn_dno_S_end {parent=dyn_dno_S, loc=<8.8707 0 0>, rot=<0 0 0 1>}
 dyn_list_cvet_1 {parent=dyn_dno_K1, loc=<10.4959 -1.9206 8.9803>, rot=<-0.0782 0.1706 0.1187 -0.975>}
 dyn_list_cvet_2 {parent=dyn_dno_K1, loc=<10.4618 1.529 9.393>, rot=<0.378 0.1686 0.0835 -0.9065>}
 dyn_list_cvet_3 {parent=dyn_dno_K1, loc=<10.5099 4.6207 8.0157>, rot=<-0.0376 0.1692 -0.0245 -0.9846>}
 dyn_list_crug_1 {parent=dyn_dno_K1, loc=<9.8215 -6.3629 5.3946>, rot=<0.2354 0.0384 0.2197 -0.946>}
 dyn_list_crug_2 {parent=dyn_dno_K1, loc=<9.3643 -6.94 0.6257>, rot=<0.6351 -0.0892 0.1586 -0.7507>}
 dyn_list_eg {parent=dyn_dno_K1, loc=<10.0523 0.2321 -3.0742>, rot=<0.1658 -0.1141 0.1495 -0.9681>}
 dyn_list_kukuruza_1 {parent=dyn_dno_K1, loc=<10.6384 4.8189 5.1202>, rot=<-0.0073 -0.0213 0.0418 -0.9989>}
 dyn_list_kukuruza_2 {parent=dyn_dno_K1, loc=<9.2277 -4.1366 -4.6154>, rot=<0.0314 -0.0104 0.012 -0.9994>}
 dyn_list_1_0 {parent=dyn_dno_K1, loc=<8.2933 3.8893 -10.8189>, rot=<-0.121 -0.2318 0.0809 -0.9618>}
 dyn_list_1_1 {parent=dyn_list_1_0, loc=<2.8702 0 0>, rot=<0 -0.0925 0 -0.9957>}
 dyn_list_1_end {parent=dyn_list_1_1, loc=<2.1034 0 0>, rot=<0 -1 0 0>}
 dyn_list_2_0 {parent=dyn_dno_K1, loc=<8.6929 -0.9405 -11.7234>, rot=<-0.0028 -0.2625 -0.0103 -0.9649>}
 dyn_list_2_1 {parent=dyn_list_2_0, loc=<2.6454 0 0>, rot=<0 -0.1202 0 -0.9927>}
 dyn_list_end {parent=dyn_list_2_1, loc=<2.2702 0 0>, rot=<0 -1 0 0>}
]
