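# Physics asset for the M_MED_Ramen backpack cosmetic, giving its dyn_ bones simple capsule and box bodies for physical secondary motion.
# The compound dyn_dno_K1 body is the hub, constrained almost rigidly to dyn_dno_K, with a fully swing- and twist-locked dyn_dno_S chain.
# Two three-bone dyn_list chains and individual topping pieces (crug, cvet, eg, kukuruza) attach to the hub, each with its own angular limit values.
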PhysicsAsset /BRCosmetics/Accessories/FORT_Backpacks/Backpack_M_MED_Ramen/Meshes/M_MED_Ramen_Pack_Physics
bodies {
 dyn_dno_K [{
   shapeType=Capsule
   length=4.8657
   radius=0.5
   shapeName=dyn_dno_K_capsule
   center=<5.6929 0.0058 -0.085>
   rotation=<rot 75.4692 -93.9302 -3.8049>
  }]
 dyn_dno_K1 [
  {shapeType=Capsule, length=0, radius=10.9536, shapeName=dyn_dno_K1_capsule}
  {shapeType=Capsule, length=7.7965, radius=5.3722, shapeName=dyn_dno_K1_capsule, center=<5.7251 3.5294 3.5987>, rotation=<rot 0 0 -48.7179>}
  {
   shapeType=Capsule
   length=7.7965
   radius=5.3722
   shapeName=dyn_dno_K1_capsule
   center=<3.8206 -6.8489 7.4183>
   rotation=<rot 15.3962 -18.2793 -51.2095>
  }
  {shapeType=Box, extentX=0.5994, extentY=16.846, extentZ=16.846, shapeName=dyn_dno_K1_box, center=<7.821 1.8749 0>, rotation=<rot 0 -4.1437 0>}
  {
   shapeType=Box
   extentX=3.4443
   extentY=5.0352
   extentZ=8.676
   shapeName=dyn_dno_K1_box
   center=<8.2299 -0.3042 -3.8304>
   rotation=<rot 0 -4.1437 0>
  }
  {
   shapeType=Box
   extentX=3.4443
   extentY=6.0325
   extentZ=2.5912
   shapeName=dyn_dno_K1_box
   center=<8.5076 -2.7453 4.5437>
   rotation=<rot 10.1351 -24.073 25.889>
  }
 ]
 dyn_dno_S [{shapeType=Capsule, length=0.505, radius=0.505, shapeName=dyn_dno_S_capsule, center=<10.7687 0 0>, rotation=<rot 0 -90 0>}]
 dyn_dno_S_end [{
   shapeType=Capsule
   length=5.076
   radius=0.6622
   shapeName=dyn_dno_S_end_capsule
   center=<0.4594 1.1012 1.0887>
   rotation=<rot -18.5367 -90 0.3555>
  }]
 dyn_list_1_0 [{
   shapeType=Capsule
   length=4.0429
   radius=0.8836
   shapeName=dyn_list_1_0_capsule
   center=<0.8313 1.7421 0.8152>
   rotation=<rot 0 -164.9704 -88.7327>
  }]
 dyn_list_1_1 [{
   shapeType=Capsule
   length=1.0844
   radius=0.8658
   shapeName=dyn_list_1_1_capsule
   center=<0.2229 0.1456 0.4907>
   rotation=<rot 0 21.977 -86.3872>
  }]
 dyn_list_1_end [{
   shapeType=Capsule
   length=0.505
   radius=0.505
   shapeName=dyn_list_1_end_capsule
   center=<-0.1997 -0.1817 0.0498>
   rotation=<rot -2.1782 -90 9.3649>
  }]
 dyn_list_2_0 [{
   shapeType=Capsule
   length=4.1411
   radius=0.8788
   shapeName=dyn_list_2_0_capsule
   center=<0.734 0.3465 0.4124>
   rotation=<rot -6.2117 -1.6683 -87.6284>
  }]
 dyn_list_2_1 [{
   shapeType=Capsule
   length=2.0906
   radius=1.0558
   shapeName=dyn_list_2_1_capsule
   center=<0.448 0.4125 0.45>
   rotation=<rot -0.5988 171.0622 -82.0267>
  }]
 dyn_list_crug_1 [{
   shapeType=Box
   extentX=0.726
   extentY=6.1851
   extentZ=6.961
   shapeName=dyn_list_crug_1_box
   center=<-0.8909 0.5359 0.0001>
   rotation=<rot -0.0061 13.465 0.0039>
  }]
 dyn_list_crug_2 [{
   shapeType=Box
   extentX=0.6946
   extentY=7.5
   extentZ=5.9017
   shapeName=dyn_list_crug_2_box
   center=<-0.0072 0.6681 0.5038>
   rotation=<rot 0 0 29.8414>
  }]
 dyn_list_cvet_1 [{shapeType=Box, extentX=0.5911, extentY=3.5958, extentZ=3.5958, shapeName=dyn_list_cvet_1_box, rotation=<rot 0 -2.6184 0>}]
 dyn_list_cvet_2 [{
   shapeType=Box
   extentX=0.5
   extentY=3.0928
   extentZ=3.0928
   shapeName=dyn_list_cvet_2_box
   center=<-0.827 0.0896 0>
   rotation=<rot 0 -6.1824 0>
  }]
 dyn_list_cvet_3 [{shapeType=Capsule, length=1.7058, radius=0.5, shapeName=dyn_list_cvet_3_capsule, center=<0.39 0 0>}]
 dyn_list_eg [
  {
   shapeType=Capsule
   length=3.45
   radius=1.3219
   shapeName=dyn_list_eg_capsule
   center=<0.2993 -0.158 -0.1847>
   rotation=<rot 4.2218 -90 -1.5363>
  }
  {
   shapeType=Capsule
   length=2.239
   radius=1.1551
   shapeName=dyn_list_eg_capsule
   center=<0.8304 0.0789 0.0694>
   rotation=<rot 0.2626 -1.7794 98.3922>
  }
 ]
 dyn_list_end [{
   shapeType=Capsule
   length=0.505
   radius=0.505
   shapeName=dyn_list_end_capsule
   center=<-0.6273 -0.121 0.3363>
   rotation=<rot 8.5108 -90 10.3457>
  }]
 dyn_list_kukuruza_1 [{
   shapeType=Capsule
   length=7.616
   radius=1.0759
   shapeName=dyn_list_kukuruza_1_capsule
   center=<-0.2611 -0.0546 -0.6041>
   rotation=<rot 0 8.4496 -63.7872>
  }]
 dyn_list_kukuruza_2 [{
   shapeType=Capsule
   length=0.5246
   radius=1.0427
   shapeName=dyn_list_kukuruza_2_capsule
   center=<-0.0937 -1.3204 -0.2804>
   rotation=<rot 0 169.7016 -73.6039>
  }]
}
constraints [
 {
  bone1Name=dyn_dno_S
  bone2Name=dyn_dno_K1
  swing1LimitDegrees=6
  swing1Motion=Locked
  swing2LimitDegrees=6
  swing2Motion=Locked
  twistLimitDegrees=0
  twistMotion=Locked
 }
 {
  bone1Name=dyn_dno_S_end
  bone2Name=dyn_dno_S
  swing1LimitDegrees=1
  swing1Motion=Locked
  swing2LimitDegrees=1
  swing2Motion=Locked
  twistLimitDegrees=0
  twistMotion=Locked
 }
 {bone1Name=dyn_list_1_0, bone2Name=dyn_dno_K1, swing1LimitDegrees=5, swing2LimitDegrees=11, twistLimitDegrees=0}
 {bone1Name=dyn_list_1_1, bone2Name=dyn_list_1_0, swing1LimitDegrees=5, swing2LimitDegrees=55, twistLimitDegrees=0}
 {bone1Name=dyn_list_1_end, bone2Name=dyn_list_1_1, swing1LimitDegrees=11, swing2LimitDegrees=34, twistLimitDegrees=0}
 {bone1Name=dyn_list_2_0, bone2Name=dyn_dno_K1, swing1LimitDegrees=5, swing2LimitDegrees=5, twistLimitDegrees=0}
 {bone1Name=dyn_list_2_1, bone2Name=dyn_list_2_0, swing1LimitDegrees=5, swing2LimitDegrees=41, twistLimitDegrees=0}
 {bone1Name=dyn_list_end, bone2Name=dyn_list_2_1, swing1LimitDegrees=11, swing2LimitDegrees=34, twistLimitDegrees=0}
 {bone1Name=dyn_list_crug_1, bone2Name=dyn_dno_K1, swing1LimitDegrees=28, swing2LimitDegrees=4, twistLimitDegrees=3}
 {bone1Name=dyn_list_crug_2, bone2Name=dyn_dno_K1, swing1LimitDegrees=4, swing2LimitDegrees=3, twistLimitDegrees=3}
 {bone1Name=dyn_list_cvet_1, bone2Name=dyn_dno_K1, swing1LimitDegrees=22, swing2LimitDegrees=22, twistLimitDegrees=3}
 {bone1Name=dyn_list_cvet_2, bone2Name=dyn_dno_K1, swing1LimitDegrees=22, swing2LimitDegrees=22, twistLimitDegrees=3}
 {bone1Name=dyn_list_eg, bone2Name=dyn_dno_K1, swing1LimitDegrees=8, swing2LimitDegrees=31, twistLimitDegrees=3}
 {bone1Name=dyn_list_kukuruza_1, bone2Name=dyn_dno_K1, swing1LimitDegrees=1, swing2LimitDegrees=1, twistLimitDegrees=1}
 {bone1Name=dyn_list_kukuruza_2, bone2Name=dyn_dno_K1, swing1LimitDegrees=5, swing2LimitDegrees=5, twistLimitDegrees=1}
 {
  bone1Name=dyn_dno_K1
  bone2Name=dyn_dno_K
  swing1LimitDegrees=0.3
  swing2LimitDegrees=0.3
  swing2Motion=Locked
  twistLimitDegrees=0.3
  twistMotion=Locked
 }
 {bone1Name=dyn_list_cvet_3, bone2Name=dyn_dno_K1, swing1LimitDegrees=45, swing2LimitDegrees=22, twistLimitDegrees=2}
]
