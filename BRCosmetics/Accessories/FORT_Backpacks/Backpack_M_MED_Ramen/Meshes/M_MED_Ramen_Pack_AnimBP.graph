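# Animation Blueprint for the Ramen backpack cosmetic, adding secondary physics and sway motion to its dyn_dno bones.
# Each update stores the owning mesh component's world location, which replaces dyn_dno_all's world translation, and computes real-time sine twist values.
# The input pose then passes through a partially blended RigidBody with overridden gravity, an additive twist on dyn_dno_K1, and the pack's Control Rig.
# The `twist` value is computed but unused in the AnimGraph.
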AnimBlueprint /BRCosmetics/Accessories/FORT_Backpacks/Backpack_M_MED_Ramen/Meshes/M_MED_Ramen_Pack_AnimBP
target_skeleton=&/M_MED_Ramen_Pack_Skeleton
graphs [
 AnimGraph {nodes [
   n1 = "Transform (Modify) Bone - Bone: dyn_dno_all"!(
    ComponentPose=Local To Component(LocalPose=Input Pose())
    Alpha=1
    Translation_X:double=Get World X() -> (World X:double)
    Translation_Y:double=Get World Y() -> (World Y:double)
    Translation_Z:double=Get world Z() -> (world Z:double)
    Rotation="0,0,0"
    Scale="1,1,1"
   ) -> (Pose) {
    rotationSpace=BCS_WorldSpace
    translationMode=BMM_Replace
    translationSpace=BCS_WorldSpace
    alphaBoolBlend {blendOption=Linear}
    boneToModify {boneName=dyn_dno_all}
    translation=<2180.022 0.992>
   }
   n2 = RigidBody(ComponentPose=@n1, Alpha=0.4) {
    bOverrideWorldGravity=true
    cachedBoundsScale=1
    overridePhysicsAsset=&/M_MED_Ramen_Pack_Physics
    alphaBoolBlend {blendOption=Linear}
    componentAppliedLinearAccClamp=<1000 1000 1000>
    componentLinearAccScale=<1 0.5 1>
    componentLinearVelScale=<1 0.5 1>
    overrideWorldGravity {z=-100}
   }
   n3 = "Transform (Modify) Bone - Bone: dyn_dno_K1"!(
    ComponentPose=@n2
    Alpha=2
    Translation="0,0,0"
    Rotation_Roll:float=Get twist2() -> (twist2:double)
    Rotation_Pitch:float=0
    Rotation_Yaw:float=Get Twist3() -> (Twist3:double)
    Scale="1,1,1"
   ) -> (Pose) {rotationMode=BMM_Additive, rotationSpace=BCS_BoneSpace, alphaBoolBlend {blendOption=Linear}, boneToModify {boneName=dyn_dno_K1}}
   n4 = Control Rig(Source=Component To Local(ComponentPose=@n3)) {
    alphaBoolBlend {blendOption=Linear}
    controlRigAssetReference {blueprintRigClass=&/M_MED_Ramen_Pack_CtrlRig.M_MED_Ramen_Pack_CtrlRig_C}
   }
   Output Pose(Result=@n4) {layerGroup=None}
   Reroute Node(InputPin:Wildcard) -> (OutputPin:Wildcard)
  ]}
 EventGraph {nodes [
   Event BlueprintUpdateAnimation()
   > n1 = Sequence!(execute) -> (then_0, then_1, then_2:Exec)
   n2 = Get World Location!(self:Scene Component Object Reference=GetOwnerSkeletalMeshComponent(self=Self-Reference())) -> (ReturnValue_X:double, ReturnValue_Y:double, ReturnValue_Z:double) {is=title K2Node_CallFunction}
   Set World X(execute=@n1.then_1, World X:double=@n2) -> (Output_Get:double)
   > Set World Y(World Y:double=@n2.ReturnValue_Y) -> (Output_Get:double)
   > Set world Z(world Z:double=@n2.ReturnValue_Z) -> (Output_Get:double)
   n3 = GetRealTimeSeconds()
   n4 = float + float(A=float * float(A=@n3, B=1) {is=title K2Node_PromotableOperator})
   Set twist(execute=@n1.then_2, twist:double=float * float(A="Sin (Radians)"(A=@n4), B=0.2)) -> (Output_Get:double)
   n5 = float + float(A=float * float(A=@n3, B=2), B=111111)
   > Set twist2(twist2:double=float * float(A="Sin (Radians)"(A=@n5), B=2)) -> (Output_Get:double)
   n6 = float + float(A=float * float(A=@n3, B=1), B=30)
   > Set Twist3(Twist3:double=float * float(A="Sin (Radians)"(A=@n6), B=0.2)) -> (Output_Get:double)
   TryGetPawnOwner()
  ]}
]
variables [
 ailpha {type=number}
 MODE {type=boolean}
 NewVar {type=Rotator}
 NewVar_0 {type=number}
 twist {type=number}
 twist2 {type=number}
 Twist3 {type=number}
 World X {type=number}
 World Y {type=number}
 world Z {type=number, default=-20000}
]
